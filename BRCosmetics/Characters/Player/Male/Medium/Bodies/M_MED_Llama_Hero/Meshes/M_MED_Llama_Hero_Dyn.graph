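# Physics asset for the M_MED_Llama_Hero male body cosmetic, covering secondary motion of its belt straps, chest pouches and shoelaces.
# Six belt strap boxes and two pouch capsules each hinge on a tiny MSTR anchor body with angular limits, with strap twist locked.
# Two-segment inner and outer lace chains branch from per-foot MSTR anchors, which link through free-rotation constraints to calf and thigh capsules.
# A separate pelvis capsule, named for collision, has no constraints.
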PhysicsAsset /BRCosmetics/Characters/Player/Male/Medium/Bodies/M_MED_Llama_Hero/Meshes/M_MED_Llama_Hero_Dyn
bodies {
 calf_twist_02_l [{shapeType=Capsule, length=22.0543, radius=4.5953, center=<-10.0262 -1.9533 0.5059>, rotation=<rot 90 0 0>}]
 calf_twist_02_r [{shapeType=Capsule, length=22.0543, radius=4.5953, center=<10.0262 1.9533 -0.5059>, rotation=<rot 90 0 0>}]
 dyn_belt_strap_bk_l [{shapeType=Box, extentX=7.5, extentY=7.5, extentZ=1.2387, center=<4.9578 0 -0.0002>}]
 dyn_belt_strap_bk_l_MSTR [{shapeType=Capsule, length=0.5, radius=0.5, rotation=<rot 90 0 0>}]
 dyn_belt_strap_bk_r [{shapeType=Box, extentX=7.5, extentY=7.5, extentZ=1.4403, center=<7.9981 -0.9498 -0.1819>, rotation=<rot -6.5523 0 0>}]
 dyn_belt_strap_bk_r_MSTR [{shapeType=Capsule, length=0.5, radius=0.5, rotation=<rot 90 0 0>}]
 dyn_belt_strap_l [{shapeType=Box, extentX=3.9035, extentY=7.5, extentZ=1.2387, center=<8.3177 0 -0.0002>}]
 dyn_belt_strap_l_MSTR [{shapeType=Capsule, length=0.5, radius=0.5, rotation=<rot 90 0 0>}]
 dyn_belt_strap_mid_l [{shapeType=Box, extentX=7.5, extentY=7.5, extentZ=1.2387, center=<10.6087 0.0323 1.4198>, rotation=<rot -3.8018 8.9085 -14.0161>}]
 dyn_belt_strap_mid_l_MSTR [{shapeType=Capsule, length=0.5, radius=0.5, rotation=<rot 90 0 0>}]
 dyn_belt_strap_mid_r [{shapeType=Box, extentX=7.5, extentY=7.5, extentZ=1.6331, center=<10.6481 -0.3759 1.7798>, rotation=<rot -9.9066 -8.5615 17.8106>}]
 dyn_belt_strap_mid_r_MSTR [{shapeType=Capsule, length=0.5, radius=0.5, rotation=<rot 90 0 0>}]
 dyn_belt_strap_r [{shapeType=Box, extentX=3.835, extentY=7.5, extentZ=1.2387, center=<8.4735 0 -0.0001>}]
 dyn_belt_strap_r_MSTR [{shapeType=Capsule, length=0.5, radius=0.5, rotation=<rot 90 0 0>}]
 dyn_pelvis_collision [{shapeType=Capsule, length=2.4175, radius=14.2779, center=<-4.0777 2.6176 0.0001>}]
 dyn_pouch_chest_01 [{shapeType=Capsule, length=8.2592, radius=4.6907, center=<0 0 3>, rotation=<rot 90 0 0>}]
 dyn_pouch_chest_01_MSTR [{shapeType=Capsule, length=0.5, radius=0.5, rotation=<rot 90 0 0>}]
 dyn_pouch_chest_02 [{shapeType=Capsule, length=8.2592, radius=4.6907, center=<0 0 3>, rotation=<rot 90 0 0>}]
 dyn_pouch_chest_02_MSTR [{shapeType=Capsule, length=0.5, radius=0.5, rotation=<rot 90 0 0>}]
 dyn_shoelace_in_l_01 [{shapeType=Capsule, length=3, radius=1, center=<1.5 0 0>, rotation=<rot 90 0 0>}]
 dyn_shoelace_in_l_02 [{shapeType=Capsule, length=3, radius=1, center=<1.5 0 0>, rotation=<rot 90 0 0>}]
 dyn_shoelace_in_r_01 [{shapeType=Capsule, length=3, radius=1, center=<1.5 0 0>, rotation=<rot 90 0 0>}]
 dyn_shoelace_in_r_02 [{shapeType=Capsule, length=3, radius=1, center=<1.5 0 0>, rotation=<rot 90 0 0>}]
 dyn_shoelace_l_MSTR [{shapeType=Capsule, length=0.2, radius=0.2, rotation=<rot 90 0 0>}]
 dyn_shoelace_out_l_01 [{shapeType=Capsule, length=3, radius=1, center=<1.5 0 0>, rotation=<rot 90 0 0>}]
 dyn_shoelace_out_l_02 [{shapeType=Capsule, length=3, radius=1, center=<1.5 0 0>, rotation=<rot 90 0 0>}]
 dyn_shoelace_out_r_01 [{shapeType=Capsule, length=3, radius=1, center=<1.5 0 0>, rotation=<rot 90 0 0>}]
 dyn_shoelace_out_r_02 [{shapeType=Capsule, length=3, radius=1, center=<1.5 0 0>, rotation=<rot 90 0 0>}]
 dyn_shoelace_r_MSTR [{shapeType=Capsule, length=0.2, radius=0.2, rotation=<rot 90 0 0>}]
 thigh_l [{shapeType=Capsule, length=14.7763, radius=8.8851, center=<-15.1355 0.7671 0.6907>, rotation=<rot 90 -12.9074 -10.5889>}]
 thigh_r [{shapeType=Capsule, length=14.7763, radius=8.8851, center=<15.1355 -0.7671 -0.5873>, rotation=<rot 90 90 92.3186>}]
}
constraints [
 {bone1Name=dyn_shoelace_out_l_01, bone2Name=dyn_shoelace_l_MSTR, swing1LimitDegrees=45, swing2LimitDegrees=45, twistLimitDegrees=10}
 {bone1Name=dyn_shoelace_out_l_02, bone2Name=dyn_shoelace_out_l_01, swing1LimitDegrees=45, swing2LimitDegrees=45, twistLimitDegrees=10}
 {bone1Name=dyn_shoelace_in_l_01, bone2Name=dyn_shoelace_l_MSTR, swing1LimitDegrees=45, swing2LimitDegrees=45, twistLimitDegrees=10}
 {bone1Name=dyn_shoelace_in_l_02, bone2Name=dyn_shoelace_in_l_01, swing1LimitDegrees=45, swing2LimitDegrees=45, twistLimitDegrees=10}
 {bone1Name=dyn_shoelace_out_r_01, bone2Name=dyn_shoelace_r_MSTR, swing1LimitDegrees=45, swing2LimitDegrees=45, twistLimitDegrees=10}
 {bone1Name=dyn_shoelace_out_r_02, bone2Name=dyn_shoelace_out_r_01, swing1LimitDegrees=45, swing2LimitDegrees=45, twistLimitDegrees=10}
 {bone1Name=dyn_shoelace_in_r_01, bone2Name=dyn_shoelace_r_MSTR, swing1LimitDegrees=45, swing2LimitDegrees=45, twistLimitDegrees=10}
 {bone1Name=dyn_shoelace_in_r_02, bone2Name=dyn_shoelace_in_r_01, swing1LimitDegrees=45, swing2LimitDegrees=45, twistLimitDegrees=10}
 {
  bone1Name=dyn_shoelace_r_MSTR
  bone2Name=calf_twist_02_r
  swing1LimitDegrees=45
  swing1Motion=Free
  swing2LimitDegrees=45
  swing2Motion=Free
  twistLimitDegrees=45
  twistMotion=Free
 }
 {
  bone1Name=dyn_shoelace_l_MSTR
  bone2Name=calf_twist_02_l
  swing1LimitDegrees=45
  swing1Motion=Free
  swing2LimitDegrees=45
  swing2Motion=Free
  twistLimitDegrees=45
  twistMotion=Free
 }
 {
  bone1Name=calf_twist_02_l
  bone2Name=thigh_l
  swing1LimitDegrees=45
  swing1Motion=Free
  swing2LimitDegrees=45
  swing2Motion=Free
  twistLimitDegrees=45
  twistMotion=Free
 }
 {
  bone1Name=calf_twist_02_r
  bone2Name=thigh_r
  swing1LimitDegrees=45
  swing1Motion=Free
  swing2LimitDegrees=45
  swing2Motion=Free
  twistLimitDegrees=45
  twistMotion=Free
 }
 {
  bone1Name=dyn_belt_strap_l
  bone2Name=dyn_belt_strap_l_MSTR
  swing1LimitDegrees=8
  swing2LimitDegrees=60
  twistLimitDegrees=45
  twistMotion=Locked
 }
 {
  bone1Name=dyn_belt_strap_bk_l
  bone2Name=dyn_belt_strap_bk_l_MSTR
  swing1LimitDegrees=15
  swing2LimitDegrees=60
  twistLimitDegrees=45
  twistMotion=Locked
 }
 {
  bone1Name=dyn_belt_strap_r
  bone2Name=dyn_belt_strap_r_MSTR
  swing1LimitDegrees=8
  swing2LimitDegrees=60
  twistLimitDegrees=45
  twistMotion=Locked
 }
 {
  bone1Name=dyn_belt_strap_bk_r
  bone2Name=dyn_belt_strap_bk_r_MSTR
  swing1LimitDegrees=15
  swing2LimitDegrees=60
  twistLimitDegrees=45
  twistMotion=Locked
 }
 {
  bone1Name=dyn_belt_strap_mid_l
  bone2Name=dyn_belt_strap_mid_l_MSTR
  swing1LimitDegrees=20
  swing2LimitDegrees=60
  twistLimitDegrees=45
  twistMotion=Locked
 }
 {
  bone1Name=dyn_belt_strap_mid_r
  bone2Name=dyn_belt_strap_mid_r_MSTR
  swing1LimitDegrees=20
  swing2LimitDegrees=60
  twistLimitDegrees=45
  twistMotion=Locked
 }
 {bone1Name=dyn_pouch_chest_01, bone2Name=dyn_pouch_chest_01_MSTR, swing1LimitDegrees=5, swing2LimitDegrees=15, twistLimitDegrees=15}
 {bone1Name=dyn_pouch_chest_02, bone2Name=dyn_pouch_chest_02_MSTR, swing1LimitDegrees=5, swing2LimitDegrees=15, twistLimitDegrees=15}
]
