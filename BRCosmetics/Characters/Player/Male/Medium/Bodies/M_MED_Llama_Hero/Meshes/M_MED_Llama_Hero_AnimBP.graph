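# Post-process style AnimBlueprint for the Llama Hero body cosmetic, adding secondary motion to bullets, pouches and antenna.
# It copies the attached parent mesh's pose, then anchors each accessory's master bone with weighted constraints to torso, pelvis and shoulder bones.
# Crouch events toggle an additive offset on the pelvis collision bone.
# AnimDynamics springs the antenna, then a pelvis-space RigidBody node simulates the Dyn physics asset with damped component motion.
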AnimBlueprint /BRCosmetics/Characters/Player/Male/Medium/Bodies/M_MED_Llama_Hero/Meshes/M_MED_Llama_Hero_AnimBP
target_skeleton=&/M_MED_Llama_Hero_Skeleton
graphs [
 AnimGraph {nodes [
   n1 = "Constraint - dyn_bullet_bk_01_MSTR (PPP)"!(
    ComponentPose=Local To Component(LocalPose=Copy Pose From Mesh() {bUseAttachedParent=true})
    Alpha=1
    ConstraintWeights_0=0.3357
    ConstraintWeights_1:float=0.0303
    ConstraintWeights_2:float=0.634
   ) -> (Pose) {
    alphaBoolBlend {blendOption=Linear}
    boneToModify {boneName=dyn_bullet_bk_01_MSTR}
    constraintSetup [{targetBone {boneName=spine_01}}, {targetBone {boneName=spine_03}}, {targetBone {boneName=pelvis}}]
    constraintWeights [1,1,1]
   }
   n2 = "Constraint - dyn_bullet_bk_02_MSTR (PPPP)"!(
    ComponentPose=@n1
    Alpha=1
    ConstraintWeights_0=0.2934
    ConstraintWeights_1:float=0.0437
    ConstraintWeights_2:float=0.0294
    ConstraintWeights_3:float=0.6335
   ) -> (Pose) {
    alphaBoolBlend {blendOption=Linear}
    boneToModify {boneName=dyn_bullet_bk_02_MSTR}
    constraintSetup [{targetBone {boneName=spine_01}}, {targetBone {boneName=spine_02}}, {targetBone {boneName=spine_03}}, {targetBone {boneName=pelvis}}]
    constraintWeights [1,1,1,1]
   }
   n3 = "Constraint - dyn_bullet_bk_03_MSTR (PPPP)"!(
    ComponentPose=@n2
    Alpha=1
    ConstraintWeights_0=0.2985
    ConstraintWeights_1:float=0.018
    ConstraintWeights_2:float=0.0193
    ConstraintWeights_3:float=0.6642
   ) -> (Pose) {
    alphaBoolBlend {blendOption=Linear}
    boneToModify {boneName=dyn_bullet_bk_03_MSTR}
    constraintSetup [{targetBone {boneName=spine_01}}, {targetBone {boneName=spine_02}}, {targetBone {boneName=spine_03}}, {targetBone {boneName=pelvis}}]
    constraintWeights [1,1,1,1]
   }
   n4 = "Constraint - dyn_bullet_bk_04_MSTR (PPP)"!(ComponentPose=@n3, Alpha=1, ConstraintWeights_0=0.3152, ConstraintWeights_1:float=0.0145, ConstraintWeights_2:float=0.6703) -> (Pose) {
    alphaBoolBlend {blendOption=Linear}
    boneToModify {boneName=dyn_bullet_bk_04_MSTR}
    constraintSetup [{targetBone {boneName=spine_01}}, {targetBone {boneName=spine_03}}, {targetBone {boneName=pelvis}}]
    constraintWeights [1,1,1]
   }
   n5 = "Constraint - dyn_bullet_chest_01_MSTR (PPPPP)"!(
    ComponentPose=@n4
    Alpha=1
    ConstraintWeights_0=0.1381
    ConstraintWeights_1:float=0.2593
    ConstraintWeights_2:float=0.2671
    ConstraintWeights_3:float=0.0538
    ConstraintWeights_4:float=0.2818
   ) -> (Pose) {
    alphaBoolBlend {blendOption=Linear}
    boneToModify {boneName=dyn_bullet_chest_01_MSTR}
    constraintSetup [
     {targetBone {boneName=spine_02}}
     {targetBone {boneName=spine_03}}
     {targetBone {boneName=spine_04}}
     {targetBone {boneName=spine_05}}
     {targetBone {boneName=clavicle_l}}
    ]
    constraintWeights [1,1,1,1,1]
   }
   n6 = "Constraint - dyn_bullet_chest_02_MSTR (PPPPPP)"!(
    ComponentPose=@n5
    Alpha=1
    ConstraintWeights_0=0.3122
    ConstraintWeights_1:float=0.1069
    ConstraintWeights_2:float=0.2072
    ConstraintWeights_3:float=0.2538
    ConstraintWeights_4:float=0.0686
    ConstraintWeights_5:float=0.0512
   ) -> (Pose) {
    alphaBoolBlend {blendOption=Linear}
    boneToModify {boneName=dyn_bullet_chest_02_MSTR}
    constraintSetup [
     {targetBone {boneName=clavicle_l}}
     {targetBone {boneName=spine_02}}
     {targetBone {boneName=spine_03}}
     {targetBone {boneName=spine_04}}
     {targetBone {boneName=spine_05}}
     {targetBone {boneName=upperarm_twist_01_l}}
    ]
    constraintWeights [1,1,1,1,1,1]
   }
   n7 = "Constraint - dyn_bullet_chest_03_MSTR (PPPPPP)"!(
    ComponentPose=@n6
    Alpha=1
    ConstraintWeights_0=0.3423
    ConstraintWeights_1:float=0.086
    ConstraintWeights_2:float=0.1753
    ConstraintWeights_3:float=0.2467
    ConstraintWeights_4:float=0.108
    ConstraintWeights_5:float=0.0416
   ) -> (Pose) {
    alphaBoolBlend {blendOption=Linear}
    boneToModify {boneName=dyn_bullet_chest_03_MSTR}
    constraintSetup [
     {targetBone {boneName=clavicle_l}}
     {targetBone {boneName=spine_02}}
     {targetBone {boneName=spine_03}}
     {targetBone {boneName=spine_04}}
     {targetBone {boneName=spine_05}}
     {targetBone {boneName=upperarm_twist_01_l}}
    ]
    constraintWeights [1,1,1,1,1,1]
   }
   n8 = "Constraint - dyn_bullet_chest_04_MSTR (PPPPPP)"!(
    ComponentPose=@n7
    Alpha=1
    ConstraintWeights_0=0.3774
    ConstraintWeights_1:float=0.0567
    ConstraintWeights_2:float=0.1308
    ConstraintWeights_3:float=0.2061
    ConstraintWeights_4:float=0.1723
    ConstraintWeights_5:float=0.0566
   ) -> (Pose) {
    alphaBoolBlend {blendOption=Linear}
    boneToModify {boneName=dyn_bullet_chest_04_MSTR}
    constraintSetup [
     {targetBone {boneName=clavicle_l}}
     {targetBone {boneName=spine_02}}
     {targetBone {boneName=spine_03}}
     {targetBone {boneName=spine_04}}
     {targetBone {boneName=spine_05}}
     {targetBone {boneName=upperarm_twist_01_l}}
    ]
    constraintWeights [1,1,1,1,1,1]
   }
   n9 = "Constraint - dyn_bullet_chest_05_MSTR (PPPPP)"!(
    ComponentPose=@n8
    Alpha=1
    ConstraintWeights_0=0.1042
    ConstraintWeights_1:float=0.1784
    ConstraintWeights_2:float=0.2622
    ConstraintWeights_3:float=0.3882
    ConstraintWeights_4:float=0.067
   ) -> (Pose) {
    alphaBoolBlend {blendOption=Linear}
    boneToModify {boneName=dyn_bullet_chest_05_MSTR}
    constraintSetup [
     {targetBone {boneName=spine_03}}
     {targetBone {boneName=spine_04}}
     {targetBone {boneName=spine_05}}
     {targetBone {boneName=clavicle_l}}
     {targetBone {boneName=upperarm_twist_01_l}}
    ]
    constraintWeights [1,1,1,1,1]
   }
   n10 = "Constraint - dyn_pouch_chest_01_MSTR (PPPPP)"!(
    ComponentPose=@n9
    Alpha=1
    ConstraintWeights_0=0.1138
    ConstraintWeights_1:float=0.305
    ConstraintWeights_2:float=0.4357
    ConstraintWeights_3:float=0.0893
    ConstraintWeights_4:float=0.0563
   ) -> (Pose) {
    alphaBoolBlend {blendOption=Linear}
    boneToModify {boneName=dyn_pouch_chest_01_MSTR}
    constraintSetup [
     {targetBone {boneName=spine_01}}
     {targetBone {boneName=spine_02}}
     {targetBone {boneName=spine_03}}
     {targetBone {boneName=spine_04}}
     {targetBone {boneName=pelvis}}
    ]
    constraintWeights [1,1,1,1,1]
   }
   n11 = "Constraint - dyn_pouch_chest_02_MSTR (PPPPP)"!(
    ComponentPose=@n10
    Alpha=1
    ConstraintWeights_0=0.1615
    ConstraintWeights_1:float=0.3548
    ConstraintWeights_2:float=0.3341
    ConstraintWeights_3:float=0.0498
    ConstraintWeights_4:float=0.0998
   ) -> (Pose) {
    alphaBoolBlend {blendOption=Linear}
    boneToModify {boneName=dyn_pouch_chest_02_MSTR}
    constraintSetup [
     {targetBone {boneName=spine_01}}
     {targetBone {boneName=spine_02}}
     {targetBone {boneName=spine_03}}
     {targetBone {boneName=spine_04}}
     {targetBone {boneName=pelvis}}
    ]
    constraintWeights [1,1,1,1,1]
   }
   n12 = "Constraint - dyn_antenna_MSTR (PPPP)"!(
    ComponentPose=@n11
    Alpha=1
    ConstraintWeights_0=0.3095
    ConstraintWeights_1:float=0.0294
    ConstraintWeights_2:float=0.0267
    ConstraintWeights_3:float=0.6344
   ) -> (Pose) {
    alphaBoolBlend {blendOption=Linear}
    boneToModify {boneName=dyn_antenna_MSTR}
    constraintSetup [{targetBone {boneName=clavicle_r}}, {targetBone {boneName=spine_03}}, {targetBone {boneName=spine_04}}, {targetBone {boneName=spine_05}}]
    constraintWeights [1,1,1,1]
   }
   n13 = "Transform (Modify) Bone - Bone: dyn_pelvis_collision"(ComponentPose=@n12, Alpha=Get crouch alpha() -> (crouch alpha:double), Translation="0,1.5,0", Rotation="0,0,0", Scale="1,1,1") {
    translationMode=BMM_Additive
    translationSpace=BCS_BoneSpace
    alphaBoolBlend {blendOption=Linear}
    boneToModify {boneName=dyn_pelvis_collision}
    translation {y=1.5}
   }
   n14 = Anim Dynamics - Bone: dyn_antenna(ComponentPose=@n13, Alpha=0.7, ExternalForce="0,0,0") {
    angularDampingOverride=0.99
    angularSpringConstant=10000
    bAngularSpring=true
    bGravityOverrideInSimSpace=true
    bOverrideAngularDamping=true
    bUseGravityOverride=true
    simulationSpace=BoneRelative
    alphaBoolBlend {blendOption=Linear}
    boundBone {boneName=dyn_antenna}
    componentAppliedLinearAccClamp=<2000 2000 2000>
    componentLinearAccScale=<1 1 1>
    componentLinearVelScale=<1 1 1>
    gravityOverride {z=-980}
    physicsBodyDefinitions [{
      sphereCollisionRadius=0
      boundBone {boneName=dyn_antenna}
      boxExtents=<10 15 15>
      constraintSetup {angularLimitsMax=<0 30 30>, angularLimitsMin=<0 -30 -30>, angularTarget=<1 0 0>}
      localJointOffset=<0.5 0 0>
     }]
    relativeSpaceBone {boneName=spine_05}
   }
   n15 = RigidBody(ComponentPose=@n14) {
    bClampLinearTranslationLimitToRefPose=true
    bOverrideWorldGravity=true
    overridePhysicsAsset=&/M_MED_Llama_Hero_Dyn
    simulationSpace=BaseBoneSpace
    alphaBoolBlend {blendOption=Linear}
    baseBoneRef {boneName=pelvis}
    componentAppliedLinearAccClamp=<3000 3000 3000>
    componentLinearAccScale=<0.35 0.3 0.3>
    componentLinearVelScale=<0.35 0.3 0.3>
    overrideWorldGravity {z=-980}
   }
   Output Pose(Result=Component To Local(ComponentPose=@n15)) {layerGroup=None}
  ]}
 EventGraph {nodes [
   Event OnBeginCrouching()
   > Set crouch alpha(crouch alpha:double=1) -> (Output_Get:double)
   Event OnEndCrouching()
   > Set crouch alpha(crouch alpha=0)
   Event BlueprintUpdateAnimation()
  ]}
]
variables [crouch alpha {type=number}]
